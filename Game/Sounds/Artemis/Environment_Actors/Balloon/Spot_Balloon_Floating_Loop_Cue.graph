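# Sound cue for the looping ambience of a floating balloon environment actor.
# It mixes a balloon float loop with a borrowed desert window wind loop, each with its own attenuation.
# Nested looping envelopers of differing lengths slowly modulate each layer's pitch and volume to add variation, and a root enveloper fades the mix in.
# All quality levels share the same mixer, and a concurrency rule limits simultaneous instances.
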SoundCue /Game/Sounds/Artemis/Environment_Actors/Balloon/Spot_Balloon_Floating_Loop_Cue
root=Enveloper_6
nodes [
 {
  id=Attenuation_0
  children [Enveloper_3]
  properties {attenuationSettings=/Game/Sounds/Artemis/POI/Seven_Workshop/Spot_Interior_Exterior_Ceiling_Wind_Focus_Att}
 }
 {
  id=Attenuation_1
  children [Enveloper_5]
  properties {attenuationSettings=&/Spot_Balloon_InAir_Att}
 }
 {
  id=Enveloper_0
  children [WavePlayer_1]
  properties {
   bLoop=true
   bLoopIndefinitely=true
   loopEnd=30
   pitchCurve {editorCurveData {keys [{interpMode=RCIM_Cubic, time=0}, <at 15.9468 0.9181 RCIM_Cubic>, {interpMode=RCIM_Cubic, time=30}]}}
   volumeCurve {editorCurveData {keys [<at 0 1>]}}
  }
 }
 {
  id=Enveloper_3
  children [Enveloper_0]
  properties {
   bLoop=true
   bLoopIndefinitely=true
   loopEnd=60
   pitchCurve {editorCurveData {keys [{time=0}]}}
   volumeCurve {editorCurveData {keys [
      {arriveTangent=-0.01, interpMode=RCIM_Cubic, time=0, value=1}
      {arriveTangent=0.0033, interpMode=RCIM_Cubic, leaveTangent=0.0033, tangentMode=RCTM_User, time=7.7071, value=1.4}
      {arriveTangent=-0.0013, interpMode=RCIM_Cubic, leaveTangent=-0.0013, tangentMode=RCTM_User, time=20.3998, value=0.3371}
      {arriveTangent=-0.004, interpMode=RCIM_Cubic, leaveTangent=-0.004, tangentMode=RCTM_User, time=30, value=1}
      <at 45 0.5 RCIM_Cubic>
      <at 60 1>
     ]}}
  }
 }
 {
  id=Enveloper_4
  children [WavePlayer_0]
  properties {
   bLoop=true
   bLoopIndefinitely=true
   loopEnd=30
   pitchCurve {editorCurveData {keys [{interpMode=RCIM_Cubic, time=0}, <at 15.9468 0.9181 RCIM_Cubic>, {interpMode=RCIM_Cubic, time=30}]}}
   volumeCurve {editorCurveData {keys [<at 0 1>]}}
  }
 }
 {
  id=Enveloper_5
  children [Enveloper_4]
  properties {
   bLoop=true
   bLoopIndefinitely=true
   loopEnd=20
   pitchCurve {editorCurveData {keys [{time=0}]}}
   volumeCurve {editorCurveData {keys [
      {arriveTangent=-0.01, interpMode=RCIM_Cubic, time=0, value=1}
      <at 10 0.8 RCIM_Cubic>
      {interpMode=RCIM_Cubic, leaveTangent=0.01, time=20, value=1}
     ]}}
  }
 }
 {
  id=Enveloper_6
  children [QualityLevel_0]
  properties {pitchCurve {editorCurveData {keys [{time=0}]}}, volumeCurve {editorCurveData {keys [<at 0 0 RCIM_Cubic>,<at 2 1>]}}}
 }
 {id=Mixer_0, children [Attenuation_0, Attenuation_1], properties {inputVolume [1,0.8]}}
 {id=QualityLevel_0, children [Mixer_0,Mixer_0,Mixer_0,Mixer_0]}
 {id=WavePlayer_0, properties {bLooping=true, soundWaveAssetPtr=&/Spot_Balloon_Floating_Loop_01}}
 {id=WavePlayer_1, properties {bLooping=true, soundWaveAssetPtr=/Game/Sounds/Artemis/POI/Seven_Workshop/Spot_Desert_Wind_Window_Loop_Stereo}}
]
properties {
 maxDistance=8000
 pitchMultiplier=0.9
 soundClassObject=&Fort_Gadgets_Items
 volumeMultiplier=0.4
 attenuationOverrides {attenuationShape=Capsule, falloffDistance=5000, reverbDistanceMin=1000, reverbWetLevelMax=3, attenuationShapeExtents {x=4000}}
 concurrencySet [/Game/Sounds/Concurrency_Rules/ThreeMaximum_Owner_PreventNew]
}
