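# Looping ambient sound cue for the energy hum of a balloon jump pad, using the trap loop attenuation and gadget sound class.
# A mixer layers a vortex wind loop with a randomly chosen UFO tech loop or alien drone, each modulated.
# Two quality levels route it through nested looping envelopes that slowly swell volume and dip pitch; the other two use the plain mix.
# A root envelope fades everything in, and concurrency prevents new instances.
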SoundCue /Game/Sounds/Artemis/Environment_Actors/Balloon/Spot_Balloon_JumpPad_Light_Energy_Loop_Cue
root=Enveloper_3
nodes [
 {
  id=Enveloper_0
  children [Enveloper_1]
  properties {
   bLoop=true
   bLoopIndefinitely=true
   loopEnd=60
   pitchCurve {editorCurveData {keys [{time=0}]}}
   volumeCurve {editorCurveData {keys [
      {arriveTangent=-0.01, interpMode=RCIM_Cubic, time=0, value=1}
      <at 10 0.5 RCIM_Cubic>
      {arriveTangent=-0.0086, interpMode=RCIM_Cubic, leaveTangent=-0.0086, time=20, value=1}
      <at 45 0.2 RCIM_Cubic>
      <at 60 1>
     ]}}
  }
 }
 {
  id=Enveloper_1
  children [Mixer_0]
  properties {
   bLoop=true
   bLoopIndefinitely=true
   loopEnd=30
   pitchCurve {editorCurveData {keys [{interpMode=RCIM_Cubic, time=0}, <at 15.9468 0.9181 RCIM_Cubic>, {interpMode=RCIM_Cubic, time=30}]}}
   volumeCurve {editorCurveData {keys [<at 0 1>]}}
  }
 }
 {
  id=Enveloper_3
  children [QualityLevel_0]
  properties {pitchCurve {editorCurveData {keys [{time=0}]}}, volumeCurve {editorCurveData {keys [<at 0 0 RCIM_Cubic>,<at 2 1>]}}}
 }
 {id=Mixer_0, children [WavePlayer_1,Random_0], properties {inputVolume [0.4,0.5]}}
 {id=Modulator_0, children [WavePlayer_2], properties {volumeMax=6.2, volumeMin=6}}
 {id=Modulator_1, children [WavePlayer_0], properties {pitchMax=1, pitchMin=0.9, volumeMax=0.9, volumeMin=0.85}}
 {id=QualityLevel_0, children [Enveloper_0,Enveloper_0,Mixer_0,Mixer_0]}
 {id=Random_0, children [Modulator_1,Modulator_0], properties {weights [1,1]}}
 {
  id=WavePlayer_0
  properties {bLooping=true, soundWaveAssetPtr=/Game/Sounds/Apollo/Biomes/UFO_Crash_Sites/AmbientLoops/AMB_UFO_Crashsite_Ambience_Tech_Loop_02}
 }
 {
  id=WavePlayer_1
  properties {bLooping=true, soundWaveAssetPtr=/Game/Sounds/Artemis/POI/RealityTree/Spot_Reality_Tree_Vortex_Wind_Loop_Stereo}
 }
 {
  id=WavePlayer_2
  properties {bLooping=true, soundWaveAssetPtr=/Game/Sounds/Apollo/Biomes/UFO_Crash_Sites/AmbientLoops/AMB_UFO_Crashsite_Loop_Alien_Drone_02}
 }
]
properties {
 attenuationSettings=/Game/Sounds/Attenuation/Fort_Traps_Loop_Att
 bOverrideConcurrency=true
 maxDistance=2500
 soundClassObject=&Fort_Gadgets
 volumeMultiplier=0.5
 concurrencyOverrides {resolutionRule=PreventNew}
}
