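# Looping positional sound cue for the lift balloon's burner thruster, routed to the Fort_Gadgets sound class with limited concurrency.
# A quality-level node picks between two branches, each a looping thruster wave under nested, endlessly looping Envelopers.
# The inner envelope modulates pitch and the outer one slowly dips volume on a different period, varying the loop over time.
# The branch for the first two quality levels dips volume deeper.
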SoundCue /Game/Sounds/Artemis/Environment_Actors/Balloon/Spot_Balloon_LifBalloon_Burner_Loop_Cue
root=QualityLevel_0
nodes [
 {
  id=Enveloper_0
  children [WavePlayer_0]
  properties {
   bLoop=true
   bLoopIndefinitely=true
   loopEnd=30
   pitchCurve {editorCurveData {keys [{interpMode=RCIM_Cubic, time=0}, <at 15.9468 0.9181 RCIM_Cubic>, {interpMode=RCIM_Cubic, time=30}]}}
   volumeCurve {editorCurveData {keys [<at 0 1>]}}
  }
 }
 {
  id=Enveloper_2
  children [Enveloper_4]
  properties {
   bLoop=true
   bLoopIndefinitely=true
   loopEnd=20
   pitchCurve {editorCurveData {keys [{time=0}]}}
   volumeCurve {editorCurveData {keys [
      {arriveTangent=-0.01, interpMode=RCIM_Cubic, time=0, value=1}
      <at 10 0.7 RCIM_Cubic>
      {interpMode=RCIM_Cubic, leaveTangent=0.01, time=20, value=1}
     ]}}
  }
 }
 {
  id=Enveloper_3
  children [Enveloper_0]
  properties {
   bLoop=true
   bLoopIndefinitely=true
   loopEnd=20
   pitchCurve {editorCurveData {keys [{time=0}]}}
   volumeCurve {editorCurveData {keys [
      {arriveTangent=-0.01, interpMode=RCIM_Cubic, leaveTangent=-0.01, time=0, value=1}
      <at 10 0.9 RCIM_Cubic>
      {arriveTangent=0.01, interpMode=RCIM_Cubic, leaveTangent=0.01, time=20, value=1}
     ]}}
  }
 }
 {
  id=Enveloper_4
  children [WavePlayer_3]
  properties {
   bLoop=true
   bLoopIndefinitely=true
   loopEnd=30
   pitchCurve {editorCurveData {keys [{interpMode=RCIM_Cubic, time=0}, <at 15.9468 0.9181 RCIM_Cubic>, {interpMode=RCIM_Cubic, time=30}]}}
   volumeCurve {editorCurveData {keys [<at 0 1>]}}
  }
 }
 {id=QualityLevel_0, children [Enveloper_2,Enveloper_2,Enveloper_3,Enveloper_3]}
 {
  id=WavePlayer_0
  properties {bLooping=true, soundWaveAssetPtr=&/Spot_Balloon_Burner_Thruster_Loop_01}
 }
 {
  id=WavePlayer_3
  properties {bLooping=true, soundWaveAssetPtr=&Spot_Balloon_Burner_Thruster_Loop_01}
 }
]
properties {
 attenuationSettings=&/Spot_Balloon_LiftBalloon_Burner_Smaller_Sphere_Att
 bOverrideConcurrency=true
 maxDistance=900
 soundClassObject=&Fort_Gadgets
 volumeMultiplier=0.4
 attenuationOverrides {attenuationShape=Capsule, falloffDistance=5000, reverbDistanceMin=1000, reverbWetLevelMax=3, attenuationShapeExtents {x=4000}}
 concurrencyOverrides {maxCount=1, voiceStealReleaseTime=5}
 concurrencySet [/Game/Sounds/Concurrency_Rules/ThreeMaximum_Owner_PreventNew]
}
